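# Looping ambient Sound Cue of structural creaks and metal movement for a big urban building foundation by day.
# The first quality branch crossfades metal and castle-creak loops through opposing 30-second bell volume envelopes.
# The second uses only the metal loop, with slow pitch and volume drift.
# Modulators randomize both branches, which play with square-map attenuation, an in-building ambience sound class and prevent-new concurrency.
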SoundCue /Figment_S07/Sounds/Fort_Ambient_Looping/Foundation_Mixes/FOUND_BuildingBig_Urban02_Day_Cue
root=QualityLevel_0
nodes [
 {
  id=Enveloper_0
  children [WavePlayer_2]
  properties {
   bLoop=true
   bLoopIndefinitely=true
   loopEnd=30
   pitchCurve {editorCurveData {keys [{interpMode=RCIM_Cubic, time=0}, <at 15.8492 0.9 RCIM_Cubic>, {interpMode=RCIM_Cubic, time=30}]}}
   volumeCurve {editorCurveData {keys [<at 0 1 RCIM_Cubic>,<at 17.3137 0.8631 RCIM_Cubic>,<at 30 1 RCIM_Cubic>]}}
  }
 }
 {
  id=Enveloper_8
  children [WavePlayer_1]
  properties {
   bLoop=true
   bLoopIndefinitely=true
   loopEnd=30
   pitchCurve {editorCurveData {keys [<at 0 1.2>]}}
   volumeCurve {externalCurve=/Figment_S07/Sounds/Trans_Enveloper_Curves/Bell_30s, editorCurveData {keys [<at 0 1>]}}
  }
 }
 {
  id=Enveloper_9
  children [WavePlayer_9]
  properties {
   bLoop=true
   bLoopIndefinitely=true
   loopEnd=30
   pitchCurve {editorCurveData {keys [{time=0}]}}
   volumeCurve {externalCurve=/Figment_S07/Sounds/Trans_Enveloper_Curves/BellReversed_30s, editorCurveData {keys [<at 0 1>]}}
  }
 }
 {id=Mixer_3, children [Enveloper_8,Enveloper_9], properties {inputVolume [1,1]}}
 {id=Modulator_0, children [Enveloper_0], properties {pitchMax=1.1, pitchMin=0.9, volumeMax=1.1, volumeMin=0.9}}
 {id=Modulator_12, children [Mixer_3], properties {pitchMax=1.2, pitchMin=0.8, volumeMax=1.1, volumeMin=0.9}}
 {id=QualityLevel_0, children [Modulator_12,Modulator_0,null,null]}
 {
  id=WavePlayer_1
  properties {bLooping=true, soundWaveAssetPtr=/Figment_S07/Sounds/Trans_AmbientLoops/StructureMovement/CounstructionMetal027_Loop_Stereo}
 }
 {
  id=WavePlayer_2
  properties {bLooping=true, soundWaveAssetPtr=&CounstructionMetal027_Loop_Stereo}
 }
 {
  id=WavePlayer_9
  properties {bLooping=true, soundWaveAssetPtr=/Game/Sounds/Trans_AmbientLoops/StructureMovement/Castle_CreaksA016_Loop_Stereo}
 }
]
properties {
 attenuationSettings=/Game/Sounds/Attenuation/Fort_Map_3x3_Square
 bOverrideConcurrency=true
 maxDistance=1780
 soundClassObject=/Figment_S07/Sounds/FortSoundClassesAndModes/Foundations_Ambience_InBuilding_PSM
 volumeMultiplier=0.5
 concurrencyOverrides {resolutionRule=PreventNew}
}
