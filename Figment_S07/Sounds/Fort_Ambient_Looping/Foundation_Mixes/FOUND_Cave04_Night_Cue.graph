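# Looping night ambience cue for the Cave04 foundation mix, played on the underground foundation ambience class with square map attenuation and PreventNew concurrency.
# The first quality branch mixes castle whispers at half volume with pitched-down urban night, shaped by opposing 100-second bell envelopes so the layers alternate.
# The second branch plays only the urban night loop with a gentle 30-second volume and pitch drift; the other two are empty.
# Modulators randomise pitch and volume on both branches.
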SoundCue /Figment_S07/Sounds/Fort_Ambient_Looping/Foundation_Mixes/FOUND_Cave04_Night_Cue
root=QualityLevel_0
nodes [
 {
  id=Enveloper_0
  children [WavePlayer_2]
  properties {
   bLoop=true
   bLoopIndefinitely=true
   loopEnd=30
   pitchCurve {editorCurveData {keys [
      {interpMode=RCIM_Cubic, time=0}
      {arriveTangent=0.0002, interpMode=RCIM_Cubic, leaveTangent=0.0002, tangentMode=RCTM_User, time=7.9875, value=1.0314}
      {arriveTangent=0, interpMode=RCIM_Cubic, leaveTangent=0, tangentMode=RCTM_User, time=20.0737, value=0.9355}
      {interpMode=RCIM_Cubic, time=30}
     ]}}
   volumeCurve {editorCurveData {keys [<at 0 1 RCIM_Cubic>,<at 15.6539 0.8429 RCIM_Cubic>,<at 30 1 RCIM_Cubic>]}}
  }
 }
 {
  id=Enveloper_52
  children [WavePlayer_230]
  properties {
   bLoop=true
   bLoopIndefinitely=true
   loopEnd=100
   pitchCurve {editorCurveData {keys [{time=0}]}}
   volumeCurve {
    externalCurve=/Game/Sounds/Trans_Enveloper_Curves/Bell_100s
    editorCurveData {keys [<at 0 1 RCIM_Cubic>,<at 10 0 RCIM_Cubic>,<at 100 1 RCIM_Cubic>]}
   }
  }
 }
 {
  id=Enveloper_53
  children [WavePlayer_1]
  properties {
   bLoop=true
   bLoopIndefinitely=true
   loopEnd=100
   pitchCurve {editorCurveData {keys [<at 0 0.8>]}}
   volumeCurve {externalCurve=/Game/Sounds/Trans_Enveloper_Curves/BellReversed_100s.BellReversed_100S, editorCurveData {keys [<at 0 1>]}}
  }
 }
 {id=Mixer_26, children [Enveloper_52,Enveloper_53], properties {inputVolume [0.5,1]}}
 {id=Modulator_0, children [Enveloper_0], properties {pitchMax=1.1, pitchMin=0.9, volumeMax=1.1, volumeMin=1}}
 {id=Modulator_68, children [Mixer_26], properties {pitchMax=1.1, pitchMin=0.9, volumeMax=1.1, volumeMin=0.9}}
 {id=QualityLevel_0, children [Modulator_68,Modulator_0,null,null]}
 {id=WavePlayer_1, properties {bLooping=true, soundWaveAssetPtr=/Figment_S07/Sounds/Fort_Ambient_Looping/Urban_Exterior/Urban_Night05}}
 {id=WavePlayer_2, properties {bLooping=true, soundWaveAssetPtr=&Urban_Night05}}
 {id=WavePlayer_230, properties {bLooping=true, soundWaveAssetPtr=/Figment_S07/Sounds/Trans_AmbientLoops/RoomTones/Castle_WhispersC004_Stereo}}
]
properties {
 attenuationSettings=/Game/Sounds/Attenuation/Fort_Map_3x3_Square
 bOverrideConcurrency=true
 maxDistance=1780
 soundClassObject=/Figment_S07/Sounds/FortSoundClassesAndModes/Foundations_Ambience_Underground_PSM
 volumeMultiplier=0.7
 concurrencyOverrides {resolutionRule=PreventNew}
}
